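# Material function that plots a one-line cross-section of a texture channel as a graph, apparently a tech-art visualization aid.
# It samples along a horizontal or vertical line at Coord, selected by static switch, and masks the channel by dot product.
# The scaled value is compared with flipped UV height, giving distance-field and nonlinear falloff outputs alongside the raw value.
# A Visualizer output overlays a derivative-sized red scan line on the texture.
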MaterialFunction /Game/Global/MaterialFunctions/TechArt/MF_CrossSection
graph {nodes [
  n1 = TextureCoordinate()
  n2 = MaterialFunctionCall(Float2=NamedRerouteDeclaration(None=FunctionInput(Preview=@n1) {bUsePreviewValueAsDefault=true, inputName=ScanUV, inputType=FunctionInput_Vector2, sortPriority=-10}) {name=InScanUV, nodeColor=<rgba 0.4531 1 0 1>}) {materialFunction=&BreakOutFloat2Components}
  n3 = FunctionInput() {inputName=Coord, inputType=FunctionInput_Scalar, previewValue=<0.5 0 0 1>}
  n4 = StaticBool() {value=true}
  n5 = StaticSwitch(
   True=MaterialFunctionCall(X=@n2.R, Y=@n3) {materialFunction=&MakeFloat2}.Result
   False=MaterialFunctionCall(X=@n3, Y=@n2.G) {materialFunction=&MakeFloat2}.Result
   Value=NamedRerouteDeclaration(None=FunctionInput(Preview=@n4) {bUsePreviewValueAsDefault=true, inputName=ScanHorizontal, inputType=FunctionInput_StaticBool, sortPriority=5}) {name=ScanHorizontal, nodeColor=<rgba 0.9922 0 1 1>}
  )
  n6 = FunctionInput(Preview=TextureObject() {texture=&AtmosphericCloudNoise01}) {inputName=Texture, inputType=FunctionInput_Texture2D, sortPriority=-5}
  n7 = FunctionInput() {bUsePreviewValueAsDefault=true, inputName=ChannelMask, inputType=FunctionInput_Vector4, sortPriority=5, previewValue=<1 0 0 0>}
  n8 = DotProduct(A=TextureSample(UVs=@n5, Tex=@n6) {samplerType=SAMPLERTYPE_LinearColor}.RGBA, B=@n7)
  FunctionOutput(None=@n8) {outputName=Raw Value, sortPriority=10}
  n9 = FunctionInput() {bUsePreviewValueAsDefault=true, inputName=Output_VerticalScale, inputType=FunctionInput_Scalar, sortPriority=10, previewValue=<0.5 0 0 1>}
  n10 = Multiply(A=ConstantBiasScale(None=@n8) {bias=-0.5, scale=2}, B=@n9)
  n11 = MaterialFunctionCall(Float2=NamedRerouteUsage()) {materialFunction=&BreakOutFloat2Components}
  n12 = ConstantBiasScale(None=StaticSwitch(True=@n11.G, False=@n11.G, Value=NamedRerouteUsage())) {bias=-0.5, scale=-2}
  FunctionOutput(None=Abs(None=Subtract(A=@n10, B=@n12))) {outputName=Distance Field Result}
  n13 = Constant() {r=-1}
  n14 = Constant() {r=1}
  FunctionOutput(None=InverseLinearInterpolate(A=If(A=@n10, B=@n12, "A > B"=@n13, "A < B"=@n14), B=@n10, Value=@n12)) {description=Falloff fades to 0 at 0-1 space edges, outputName=Nonlinear Result, sortPriority=5}
  n15 = NamedRerouteUsage()
  n16 = TextureSample(UVs=@n15, Tex=@n6) {samplerType=SAMPLERTYPE_LinearColor}
  n17 = TextureCoordinate()
  n18 = DDX(Value=@n17)
  n19 = DDY(Value=@n17)
  n20 = Max(A=Multiply(A=Min(A=@n18, B=@n19)) {constB=4}) {constB=0.01}
  FunctionOutput(None=LinearInterpolate(
    A=DotProduct(A=@n16.RGBA, B=@n7)
    B=Constant3Vector() {constant=<rgba 1 0 0 0>}
    Alpha=SphereMask(A=@n5, B=@n15, Radius=@n20) {attenuationRadius=0.1}
   )) {outputName=Visualizer, sortPriority=50}
 ]}
