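# Unlit MaterialGym example of MF_CrossSection, animating a scan back and forth through a grayscale texture, blurred noise by default.
# Time drives a triangle-wave coordinate into the function, with parameters for scan axis, channel and vertical scale.
# The normal view adds a tinted glow to a distance-field line kept visible by derivative-based minimum thickness.
# The Debug View switch instead tiles the function's nonlinear, visualizer and raw outputs under an overlay texture.
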Material /Game/Global/Materials/MaterialGym/Materials/CrossSection/M_CrossSectionExample
graph {
 properties {bAutomaticallySetUsageInEditor=false, bHasPixelAnimation=true, shadingModel=MSM_Unlit, usedShadingModels=Unlit}
 nodes [
  Constant3Vector() {constant=<rgba 4 0.2 0 0>}
  n1 = TextureObjectParameter() {parameterName=Texture, samplerType=SAMPLERTYPE_Grayscale, texture=/FortUILibrary/UGC/Textures/T_UEFN_LowResBlurredNoise}
  n2 = ScalarParameter() {defaultValue=4, parameterName=ScanPeriod}
  n3 = Fmod(A=Time() {period=8}, B=Multiply(A=@n2) {constB=2})
  n4 = Abs(None=Subtract(A=Divide(A=@n3, B=@n2) {constB=4}))
  n5 = StaticBoolParameter() {defaultValue=true, parameterName=ScanHorizontal}
  n6 = VectorParameter() {parameterName=ChannelMask, defaultValue=<rgba 1 0 0 0>}
  n7 = AppendVector(A=@n6.RGB, B=@n6.RGB)
  n8 = ScalarParameter() {defaultValue=0.5, parameterName=VerticalScale}
  n9 = MaterialFunctionCall!(
   ScanUV=Frac(None=TextureCoordinate() {uTiling=2, vTiling=2})
   Texture=@n1
   Coord=@n4
   ScanHorizontal=@n5
   ChannelMask=@n7
   Output_VerticalScale=@n8
  ) {materialFunction=/Game/Global/MaterialFunctions/TechArt/MF_CrossSection}
  n10 = MaterialFunctionCall(Float2=TextureCoordinate()) {materialFunction=&BreakOutFloat2Components}
  n11 = Floor(None=Add(A=Frac(None=@n10.R)) {constB=0.5})
  n12 = Frac(None=@n10.G)
  n13 = LinearInterpolate(
   A=MaterialFunctionCall!(sRGB=LinearInterpolate(A=@n9.Nonlinear Result, B=@n9.Nonlinear Result, Alpha=@n11)) {materialFunction=/Engine/Functions/Engine_MaterialFunctions01/ImageAdjustment/sRGBToLinear}.Linear
   B=LinearInterpolate(
    A=@n9.Visualizer
    B=MaterialFunctionCall(sRGB=@n9.Raw Value) {materialFunction=&sRGBToLinear}.Linear
    Alpha=@n11
   )
   Alpha=Floor(None=Add(A=@n12) {constB=0.5}) {desc=Vertical Blend}
  )
  n14 = TextureSample() {texture=&/T_CrossSectionMFExample_Overlay}
  n15 = MaterialFunctionCall(Texture=@n1, Coord=@n4, ScanHorizontal=@n5, ChannelMask=@n7, Output_VerticalScale=@n8) {materialFunction=&MF_CrossSection}
  n16 = MaterialFunctionCall(sRGB=@n15.Nonlinear Result) {materialFunction=&sRGBToLinear}
  n17 = VectorParameter() {parameterName=GlowColor, defaultValue=<rgba 0 0.2373 1 1>}
  n18 = TextureCoordinate()
  n19 = Max(A=DDX(Value=@n18), B=DDY(Value=@n18))
  n20 = If(
   A=@n15.Distance Field Result
   B=Max(A=ScalarParameter() {defaultValue=0.01, parameterName=DF Line Thickness}, B=Multiply(A=@n19) {constB=2})
   "A > B"=Constant()
   "A < B"=Constant() {r=1}
  )
  n21 = VectorParameter() {parameterName=LineColor, defaultValue=<rgba 0.3125 0.8783 1 1>}
  StaticSwitchParameter(
   True=LinearInterpolate(A=LinearInterpolate(A=@n13, Alpha=@n14.R), Alpha=@n14.G) {constB=0}
   False=Add(A=Multiply(A=@n16.Linear, B=@n17.RGB), B=Multiply(A=@n20, B=@n21.RGB))
  ) {parameterName=Debug View}
 ]
}
